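; Blueprint function library that supplies colors, icons and style flags for Athena match objective HUD widgets, covering match quests, bounties and Job Board quests.
; GetStylingTypeFromTag maps quest gameplay tags to a style name, collapsing most types to JobBoard unless separate styles are enabled.
; Getters read that row from the category styling DataTable, mostly falling back to Default when missing.
; Urgent quest data can override main background color and alert icon.
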(fn GetMatchQuestCategoryIcon
  (params (Event%20Subtitle :type Text%20%28by%20ref%29))
  ;[1 0.177 0] This is gross. Will fix later. Hopefully.
  (if (Utilities|Text|EqualExactly(Text) (LocalVariables|GetExploration) Event%20Subtitle)
    (return (LocalVariables|GetExplorationIcon))
    (else
      (if (Utilities|Text|EqualExactly(Text) Event%20Subtitle (LocalVariables|GetCombat))
        (return (LocalVariables|GetCombatIcon))
        (else
          (return ""))))))

(fn GetFullStyingInfoFromType
  (params (Type :type Name))
  (bind _returnvalue (Utilities|GetDataTableRow "/Game/Athena/HUD/Discoverability/MatchObjectivesWidgets/Styling/DT_Athena_Objectives_CategoryStyling.DT_Athena_Objectives_CategoryStyling" Type)
    (:then
      ;[0.01938 0.4793 0.956] Returns all styling data for that row
      (return _returnvalue))
    (:RowNotFound
      ;[0.01938 0.4793 0.956] Return default styling, if a category cannot be found
      (bind _returnvalue_2 (Utilities|GetDataTableRow "/Game/Athena/HUD/Discoverability/MatchObjectivesWidgets/Styling/DT_Athena_Objectives_CategoryStyling.DT_Athena_Objectives_CategoryStyling" "Default"))
      (return _returnvalue_2))))

(fn GetIconBGColors
  (params (Type :type Name))
  (bind _returnvalue (Utilities|GetDataTableRow "/Game/Athena/HUD/Discoverability/MatchObjectivesWidgets/Styling/DT_Athena_Objectives_CategoryStyling.DT_Athena_Objectives_CategoryStyling" Type)
    (:then
      (return (Utilities|Struct|BreakStructAthenaObjectivesStylingData _returnvalue) (.IconBGColor02_16_1D6782A54B531C86AE9165B1EA22DFA7 (Utilities|Struct|BreakStructAthenaObjectivesStylingData _returnvalue))))
    (:RowNotFound
      ;[0.01938 0.4793 0.956] Return default styling, if a category cannot be found
      (bind _returnvalue_2 (Utilities|GetDataTableRow "/Game/Athena/HUD/Discoverability/MatchObjectivesWidgets/Styling/DT_Athena_Objectives_CategoryStyling.DT_Athena_Objectives_CategoryStyling" "Default"))
      (return (Utilities|Struct|BreakStructAthenaObjectivesStylingData _returnvalue_2) (.IconBGColor02_16_1D6782A54B531C86AE9165B1EA22DFA7 (Utilities|Struct|BreakStructAthenaObjectivesStylingData _returnvalue_2))))))

(fn GetIntro/OutroTitleColors
  (params (Type :type Name))
  (bind _returnvalue (Utilities|GetDataTableRow "/Game/Athena/HUD/Discoverability/MatchObjectivesWidgets/Styling/DT_Athena_Objectives_CategoryStyling.DT_Athena_Objectives_CategoryStyling" Type)
    (:then
      (return (Utilities|Struct|BreakStructAthenaObjectivesStylingData _returnvalue) (.IntroShineColor_43_1AB24B18437B9CF5593729AE337E1176 (Utilities|Struct|BreakStructAthenaObjectivesStylingData _returnvalue)) (.SuccessShineColor_46_0B4124C1429557DBDB044680E5457479 (Utilities|Struct|BreakStructAthenaObjectivesStylingData _returnvalue)) (.FailedShineColor_44_34D4636B4C0BA28F9EC7078A8F69B233 (Utilities|Struct|BreakStructAthenaObjectivesStylingData _returnvalue))))
    (:RowNotFound
      ;[0.01938 0.4793 0.956] Return default styling, if a category cannot be found
      (bind _returnvalue_2 (Utilities|GetDataTableRow "/Game/Athena/HUD/Discoverability/MatchObjectivesWidgets/Styling/DT_Athena_Objectives_CategoryStyling.DT_Athena_Objectives_CategoryStyling" "Default"))
      (return (Utilities|Struct|BreakStructAthenaObjectivesStylingData _returnvalue_2) (.IntroShineColor_43_1AB24B18437B9CF5593729AE337E1176 (Utilities|Struct|BreakStructAthenaObjectivesStylingData _returnvalue_2)) (.SuccessShineColor_46_0B4124C1429557DBDB044680E5457479 (Utilities|Struct|BreakStructAthenaObjectivesStylingData _returnvalue_2)) (.FailedShineColor_44_34D4636B4C0BA28F9EC7078A8F69B233 (Utilities|Struct|BreakStructAthenaObjectivesStylingData _returnvalue_2))))))

(fn GetMainBGColor
  (params (Type :type Name) (UrgentQuestData :type Urgent%20Quest%20Data%20Structure))
  (bind (_visualstyletag _boverridewidgetprimarycolor _overriddenwidgetprimarycolor) (Utilities|Struct|BreakUrgentQuestData UrgentQuestData))
  (bind _returnvalue (Utilities|GetDataTableRow "/Game/Athena/HUD/Discoverability/MatchObjectivesWidgets/Styling/DT_Athena_Objectives_CategoryStyling.DT_Athena_Objectives_CategoryStyling" Type)
    (:then
      (return (select _boverridewidgetprimarycolor _overriddenwidgetprimarycolor (Utilities|Struct|BreakStructAthenaObjectivesStylingData _returnvalue)) (select _boverridewidgetprimarycolor _overriddenwidgetprimarycolor (.TitleBGColor02_7_34E838C6448EF77FE3D41A9ABE1F227C (Utilities|Struct|BreakStructAthenaObjectivesStylingData _returnvalue)))))
    (:RowNotFound
      ;[0.01938 0.4793 0.956] Return default styling, if a category cannot be found
      (bind _returnvalue_2 (Utilities|GetDataTableRow "/Game/Athena/HUD/Discoverability/MatchObjectivesWidgets/Styling/DT_Athena_Objectives_CategoryStyling.DT_Athena_Objectives_CategoryStyling" "Default"))
      (return (Utilities|Struct|BreakStructAthenaObjectivesStylingData _returnvalue_2) (.TitleBGColor02_7_34E838C6448EF77FE3D41A9ABE1F227C (Utilities|Struct|BreakStructAthenaObjectivesStylingData _returnvalue_2))))))

(fn GetMatchIcon
  (params (UrgentQuestData :type Urgent%20Quest%20Data%20Structure))
  (bind (_eventsubtitle _alerticon) (Utilities|Struct|BreakUrgentQuestData UrgentQuestData))
  ; Override
  (if (Utilities|IsValidSoftObjectReference _alerticon)
    (return _alerticon)
    (else
      (return (MatchQuests-TEMP|GetMatchQuestCategoryIcon _eventsubtitle)))))

(fn GetMatchQuestTitleColors
  (return (LocalVariables|GetColor%2001) (LocalVariables|GetColor%2002)))

(fn GetMaterialOverride
  (params (Type :type Name))
  (bind _returnvalue (Utilities|GetDataTableRow "/Game/Athena/HUD/Discoverability/MatchObjectivesWidgets/Styling/DT_Athena_Objectives_CategoryStyling.DT_Athena_Objectives_CategoryStyling" Type)
    (:then
      (return (Utilities|Struct|BreakStructAthenaObjectivesStylingData _returnvalue)))
    (:RowNotFound
      (return 0))))

(fn GetObjective&ProgressBGColor
  (params (Type :type Name))
  (bind _returnvalue (Utilities|GetDataTableRow "/Game/Athena/HUD/Discoverability/MatchObjectivesWidgets/Styling/DT_Athena_Objectives_CategoryStyling.DT_Athena_Objectives_CategoryStyling" Type)
    (:then
      (return (.ProgressBGColor_68_A1C38DBD4F381B874EA914A3CA750D91 (Utilities|Struct|BreakStructAthenaObjectivesStylingData _returnvalue))))
    (:RowNotFound
      ;[0.01938 0.4793 0.956] Return default styling, if a category cannot be found
      (bind _returnvalue_2 (Utilities|GetDataTableRow "/Game/Athena/HUD/Discoverability/MatchObjectivesWidgets/Styling/DT_Athena_Objectives_CategoryStyling.DT_Athena_Objectives_CategoryStyling" "Default"))
      (return (.ProgressBGColor_68_A1C38DBD4F381B874EA914A3CA750D91 (Utilities|Struct|BreakStructAthenaObjectivesStylingData _returnvalue_2))))))

(fn GetProgressBarColor
  (params (Type :type Name))
  (bind _returnvalue (Utilities|GetDataTableRow "/Game/Athena/HUD/Discoverability/MatchObjectivesWidgets/Styling/DT_Athena_Objectives_CategoryStyling.DT_Athena_Objectives_CategoryStyling" Type)
    (:then
      (return (.ProgressBarStyle_62_0AB76DD74E7B1E2C7CE40AA4FFD263AE (Utilities|Struct|BreakStructAthenaObjectivesStylingData _returnvalue))))
    (:RowNotFound
      ;[0.01938 0.4793 0.956] Return default styling, if a category cannot be found
      (bind _returnvalue_2 (Utilities|GetDataTableRow "/Game/Athena/HUD/Discoverability/MatchObjectivesWidgets/Styling/DT_Athena_Objectives_CategoryStyling.DT_Athena_Objectives_CategoryStyling" "Default"))
      (return (.ProgressBarStyle_62_0AB76DD74E7B1E2C7CE40AA4FFD263AE (Utilities|Struct|BreakStructAthenaObjectivesStylingData _returnvalue_2))))))

(fn GetStylingTypeFromTag
  (params (UrgentQuestData :type Urgent%20Quest%20Data%20Structure))
  (bind _stylingtypes (LocalVariables|GetStylingTypes))
  (bind _array_element (Utilities|Array|ForEachLoopwithBreak (GameplayTags|BreakGameplayTagContainer (Utilities|Struct|BreakUrgentQuestData UrgentQuestData)))
    (:LoopBody
      (if (.ReturnValue (Utilities|Map|Find _stylingtypes _array_element))
        _))
    (:Completed
      _))
  (if (.ReturnValue (Utilities|Map|Find _stylingtypes _array_element))
    (LocalVariables|SetCurrentStyle (Utilities|Map|Find _stylingtypes _array_element))
    ;[1 0.177 0] For testing:
    ; Locally set and default off, until we test and determine if we want to use different styling for each invidiaul type of Job Board quest.
    ; (not including bounties)
    (if (LocalVariables|GetUseSeperateJobBoardStyles)
      ;[0.2747 0 1] Returns the specific style for unique styling colors
      (return (LocalVariables|GetCurrentStyle) (== (LocalVariables|GetCurrentStyle) "Bounty"))
      (else
        ;[0.2656 0.04701 0] Returns the default Job Board styling for anything other than Match Quests and Bounties (Adding an extra exception for Doom Quests)
        (return (Utilities|String|SelectName (LocalVariables|GetCurrentStyle) "JobBoard" (Math|Boolean|ORBoolean (== (LocalVariables|GetCurrentStyle) "MatchQuest") (== (LocalVariables|GetCurrentStyle) "Bounty") (== (LocalVariables|GetCurrentStyle) "LodeSwanQuests"))) (== (LocalVariables|GetCurrentStyle) "Bounty"))))
    (else
      (return "Default" false))))

(fn GetTextColor
  (params (Type :type Name))
  (bind _returnvalue (Utilities|GetDataTableRow "/Game/Athena/HUD/Discoverability/MatchObjectivesWidgets/Styling/DT_Athena_Objectives_CategoryStyling.DT_Athena_Objectives_CategoryStyling" Type)
    (:then
      (return (Utilities|Struct|BreakStructAthenaObjectivesStylingData _returnvalue) (.TimerTextColor_52_D027004840456C17B0B258A242B9002C (Utilities|Struct|BreakStructAthenaObjectivesStylingData _returnvalue))))
    (:RowNotFound
      ;[0.01938 0.4793 0.956] Return default styling, if a category cannot be found
      (bind _returnvalue_2 (Utilities|GetDataTableRow "/Game/Athena/HUD/Discoverability/MatchObjectivesWidgets/Styling/DT_Athena_Objectives_CategoryStyling.DT_Athena_Objectives_CategoryStyling" "Default"))
      (return (Utilities|Struct|BreakStructAthenaObjectivesStylingData _returnvalue_2) (.TimerTextColor_52_D027004840456C17B0B258A242B9002C (Utilities|Struct|BreakStructAthenaObjectivesStylingData _returnvalue_2))))))

(fn IsFullBodyIcon
  (params (Type :type Name))
  (bind _returnvalue (Utilities|GetDataTableRow "/Game/Athena/HUD/Discoverability/MatchObjectivesWidgets/Styling/DT_Athena_Objectives_CategoryStyling.DT_Athena_Objectives_CategoryStyling" Type)
    (:then
      (return (.isFullBodyIcon_70_58D57CEA46C4848E0540CFB09BF1A016 (Utilities|Struct|BreakStructAthenaObjectivesStylingData _returnvalue))))
    (:RowNotFound
      (return false))))
